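# Movie Render Graph config that renders small square animation thumbnails for the Animation Library as EXR sequences.
# The Globals branch chains motion-blur-emulating warm-up, game overrides that disable LODs and flush streaming, and low-resolution output settings.
# Start and End variables set the playback range, and the output settings node carries a "dirNode" script tag beside a hardcoded temp directory.
# The layer branch feeds a deferred renderer into single-layer EXR output, named by the outputname variable.
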MovieGraphConfig /AnimationLibrary/resources/Thumbnail_QuickRenderConfig
members {
 inputs [
  MovieGraphInput_0 {class=MovieGraphInput}
  MovieGraphInput_1 {class=MovieGraphInput}
 ]
 outputs [
  MovieGraphOutput_0 {class=MovieGraphOutput}
  MovieGraphOutput_1 {class=MovieGraphOutput}
 ]
 variables [
  MovieGraphVariable_0 {class=MovieGraphVariable}
  MovieGraphVariable_1 {class=MovieGraphVariable}
  MovieGraphVariable_2 {class=MovieGraphVariable}
  MovieGraphVariable_3 {class=MovieGraphVariable}
 ]
}
ids=id
nodes [
 DefaultInputNode = Input() {
  is=title MovieGraphInputNode
  pins are=other
  also {
   outputs [
    MovieGraphPin_1 {properties {type=None, allow_multiple_connections=false, is_branch=true, label=Globals}}
    MovieGraphPin_0 {properties {type=DOUBLE, allow_multiple_connections=false, is_branch=true, label=DefaultLayer}}
   ]
   position [-416,0]
  }
 }
 n1 = Global Game Overrides(MovieGraphPin_0=Warm Up Settings(MovieGraphPin_0=@DefaultInputNode.Globals) {
   is=title MovieGraphWarmUpSettingNode
   pins are=other
   bEmulateMotionBlur=true
   bOverride_NumWarmUpFrames=true
   bOverride_bEmulateMotionBlur=true
   numWarmUpFrames=32
   also {
    inputs [MovieGraphPin_0 {properties {type=None, allow_multiple_connections=false, is_branch=true}}]
    outputs [MovieGraphPin_1 {properties {type=None, allow_multiple_connections=false, is_branch=true}}]
    position [-256,0]
   }
  }) {
  is=title MovieGraphGlobalGameOverridesNode
  pins are=other
  bDisableHLODs=true
  bDisableLODs=true
  bFlushAssetCompiler=true
  bFlushGrassStreaming=true
  bFlushLevelStreaming=true
  bFlushShaderCompiler=true
  bFlushStreamingManagers=true
  bOverride_GameModeOverride=true
  bOverride_ScalabilityQualityLevel=true
  bOverride_VirtualTextureFeedbackFactor=true
  bOverride_bDisableHLODs=true
  bOverride_bDisableLODs=true
  bOverride_bDisableTextureStreaming=true
  bOverride_bFlushAssetCompiler=true
  bOverride_bFlushGrassStreaming=true
  bOverride_bFlushLevelStreaming=true
  bOverride_bFlushShaderCompiler=true
  bOverride_bFlushStreamingManagers=true
  also {
   inputs [MovieGraphPin_0 {properties {type=None, allow_multiple_connections=false, is_branch=true}}]
   outputs [MovieGraphPin_1 {properties {type=None, allow_multiple_connections=false, is_branch=true}}]
   position [-16,0]
  }
 }
 n2 = Global Output Settings(
  MovieGraphPin_0=@n1
  CustomPlaybackRangeStart=Start() {
   is=title MovieGraphVariableNode
   pins are=other
   also {
    outputs [MovieGraphPin_0 {properties {type=STRUCT, allow_multiple_connections=false, label=Start, type_object=/Script/MovieRenderPipelineCore.MovieGraphSequencePlaybackRangeBound}}]
    position [-16,144]
   }
  }.Start
  CustomPlaybackRangeEnd=End() {
   is=title MovieGraphVariableNode
   pins are=other
   also {
    outputs [MovieGraphPin_0 {properties {type=STRUCT, allow_multiple_connections=false, label=End, type_object=&MovieGraphSequencePlaybackRangeBound}}]
    position [-16,224]
   }
  }.End
 ) {
  is=title MovieGraphGlobalOutputSettingNode
  pins are=other
  bOverride_CustomPlaybackRangeEnd=true
  bOverride_CustomPlaybackRangeStart=true
  bOverride_OutputDirectory=true
  bOverride_OutputFrameRate=true
  bOverride_OutputResolution=true
  outputDirectory {path="D:\\Temp\\Assets\\thumbnails\\ALA_956ea7a4-b699-4234-be5e-5a1085c464ec"}
  outputFrameRate {numerator=30}
  outputResolution {profileName=Custom, resolution=<256 256>}
  scriptTags [dirNode]
  also {
   inputs [
    MovieGraphPin_0 {properties {type=None, allow_multiple_connections=false, is_branch=true}}
    MovieGraphPin_3 {properties {
      type=STRUCT
      allow_multiple_connections=false
      label=CustomPlaybackRangeStart
      type_object=&MovieGraphSequencePlaybackRangeBound
     }}
    MovieGraphPin_4 {properties {
      type=STRUCT
      allow_multiple_connections=false
      label=CustomPlaybackRangeEnd
      type_object=&MovieGraphSequencePlaybackRangeBound
     }}
   ]
   outputs [MovieGraphPin_1 {properties {type=None, allow_multiple_connections=false, is_branch=true}}]
   position [240,0]
  }
 }
 MovieGraphDeferredRenderPassNode_2 = Deferred Renderer(MovieGraphPin_0=@DefaultInputNode.DefaultLayer) {
  is=title MovieGraphDeferredRenderPassNode
  pins are=other
  showFlags=&self:MovieGraphDeferredRenderPassNode_2.ShowFlags
  also {
   inputs [MovieGraphPin_0 {properties {type=None, allow_multiple_connections=false, is_branch=true}}]
   outputs [MovieGraphPin_1 {properties {type=None, allow_multiple_connections=false, is_branch=true}}]
   position [-256,336]
  }
 }
 n3 = outputname() {
  is=title MovieGraphVariableNode
  pins are=other
  also {outputs [MovieGraphPin_0 {properties {type=STRING, allow_multiple_connections=false, label=outputname}}], position [-384,416]}
 }
 DefaultOutputNode = Output(
  Globals=@n2
  DefaultLayer=Render Layer(MovieGraphPin_0=".exr Sequence (Single-Layer)"(MovieGraphPin_0=@MovieGraphDeferredRenderPassNode_2, FileNameFormat=@n3.outputname) {
    is=title MovieGraphImageSequenceOutputNode_EXR
    pins are=other
    bOverride_FileNameFormat=true
    fileNameFormat="{sequence_name}.{frame_number}"
    also {
     inputs [
      MovieGraphPin_0 {properties {type=None, allow_multiple_connections=false, is_branch=true}}
      MovieGraphPin_2 {properties {type=STRING, allow_multiple_connections=false, label=FileNameFormat}}
     ]
     outputs [MovieGraphPin_1 {properties {type=None, allow_multiple_connections=false, is_branch=true}}]
     position [-16,336]
    }
   }) {
   is=title MovieGraphRenderLayerNode
   pins are=other
   bOverride_LayerName=true
   layerName=Layer1
   also {
    inputs [MovieGraphPin_0 {properties {type=None, allow_multiple_connections=false, is_branch=true}}]
    outputs [MovieGraphPin_1 {properties {type=None, allow_multiple_connections=false, is_branch=true}}]
    position [288,320]
   }
  }
 ) {
  is=title MovieGraphOutputNode
  pins are=other
  also {
   inputs [
    MovieGraphPin_1 {properties {type=None, allow_multiple_connections=false, is_branch=true, label=Globals}}
    MovieGraphPin_0 {properties {type=DOUBLE, allow_multiple_connections=false, is_branch=true, label=DefaultLayer}}
   ]
   position [640,0]
  }
 }
]
